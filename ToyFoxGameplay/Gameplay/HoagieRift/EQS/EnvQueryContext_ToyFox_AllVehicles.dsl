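; EQS query context for ToyFox HoagieRift that supplies every vehicle, plus every HoagieRift spawner, as context actors for environment queries.
; ProvideActorsSet gathers all FortAthenaVehicle actors and all B_ToyFox_HoagieRift_Spawner actors in the world and returns them as one combined array.
; It does not use the querier, so the result is world-wide, with no filtering by distance or state.
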(fn ProvideActorsSet
  (params (QuerierObject :type Object%20Reference) (QuerierActor :type Actor%20Object%20Reference))
  (bind _outactors (Actor|GetAllActorsOfClass "/Script/FortniteGame.FortAthenaVehicle"))
  (bind _output_get (LocalVariables|SetRetArray _outactors))
  (bind _outactors_2 (Actor|GetAllActorsOfClass "/ToyFoxGameplay/Gameplay/HoagieRift/B_ToyFox_HoagieRift_Spawner.B_ToyFox_HoagieRift_Spawner_C"))
  (Utilities|Array|AppendArray _output_get _outactors_2)
  (return _output_get))
